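; Athena mounted turret weapon that hides its mesh on BeginPlay and adds snowman anti-exploit handling and dual-gun effects.
; Vehicle setup excludes the player's attached balloon actors from weapon traces.
; A server check from BeginPlay re-polls snowman-tagged players on a timer while the turret or no weapon is current, stripping the snowman on switching to non-utility weapons.
; Muzzle and shell events spawn particles at left or right mesh sockets, unless flagged to skip.
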(event EventBeginPlay
  (|CallParentFunction)
  (Rendering|SetVisibility (Variables|Visual|GetWeaponMesh) false false) ; Hack so weapon isn't visibile until in hand
  (CallFunction|ServerMountedTurretSnowmanCheck))

(event Setup|EventOnHostVehicleSetup
  ;[1 1 1] If we have a sneaky snowman, we'll want to ignore it in our traces
  (bind _asfort_player_pawn (Utilities|Casting|CastToFortPlayerPawn (Game|GetInstigator)))
  (for _array_element (Actor|GetAttachedActors _asfort_player_pawn true false)
    (bind _asbuilding_gameplay_actor_balloon (Utilities|Casting|CastToBuildingGameplayActorBalloon _array_element))
    (WeaponTrace|AddtoIgnoreActors _asbuilding_gameplay_actor_balloon)))

(event FortWeapon|EventOnLocalUnEquip
  ;[1 1 1] EVENT -> Unequip - Do the snowman check for exploit
  (|CallParentFunction))

(event Custom|ServerMountedTurretSnowmanCheck
  (bind _playerpawn (Variables|MountedTurret|GetPlayerPawn))
  (bind _currentweapon (Variables|/Script/FortniteGame.FortPawn|GetCurrentWeapon _playerpawn))
  (bind (_asb_mounted_turret_athena _bsuccess) (Utilities|Casting|CastToB_MountedTurretAthena _currentweapon))
  ;[1 1 1] Check to see if they are trying to exploit the poor snowman - Fix for FORT-145087 (For some reason this doesn't work if I run the logic in the parent)
  (Networking|SwitchHasAuthority
    (:Authority
      (if (GameplayTags|HasMatchingGameplayTag (Variables|MountedTurret|GetPlayerPawn) (Variables|Default|GetSnowmanCue))
        (if (or _bsuccess (Utilities|Equal(Object) _currentweapon 0))
          (Utilities|Time|SetTimerbyEvent (EventDispatchers|CreateEvent :!event Custom|ServerMountedTurretSnowmanCheck) 0.1)
          (else
            (Utilities|Casting|CastToB_UtilityItem_Generic _currentweapon
              (:CastFailed
                (GameplayEffects|RemoveActiveEffectswithGrantedTags (Ability|GetAbilitySystemComponent _playerpawn) "(GameplayTags=((TagName=\"Athena.Item.Snowman\")))")
                (for _array_element (Actor|GetAttachedActors _playerpawn true false)
                  (bind _asbuilding_gameplay_actor_balloon (Utilities|Casting|CastToBuildingGameplayActorBalloon _array_element))
                  (Utilities|IsValid _asbuilding_gameplay_actor_balloon
                    (:Is%20Valid
                      (Actor|DestroyActor _asbuilding_gameplay_actor_balloon))))))))))))

(event EventMuzzleFlashFX
  (params (Persistent%20Fire :type Boolean) (OffHand :type Boolean))
  (bind _weaponmesh (Variables|Visual|GetWeaponMesh))
  (bind (_returnvalue_location _returnvalue_rotation) (Math|Transform|BreakTransform (Transformation|GetSocketTransform _weaponmesh "muzzle" "RTS_World")))
  (bind _returnvalue (CallFunction|GetLeftHandWeaponMesh))
  (bind (_returnvalue_location_2 _returnvalue_rotation_2) (Math|Transform|BreakTransform (Transformation|GetSocketTransform _returnvalue "muzzle" "RTS_World")))
  ;[1 1 1] Muzzle FX
  (if (Variables|Default|GetSkipCustomMuzzleFXSpawnScript)
    (|CallParentFunction Persistent%20Fire OffHand)
    (elif OffHand
      (bind _returnvalue_2 (Effects|SpawnEmitteratLocation "/Game/Gadgets/Assets/Turret_Gun/Particles/P_TurretGun_MFlash_L.P_TurretGun_MFlash_L" _returnvalue_location_2 _returnvalue_rotation_2 "1.000000,1.000000,1.000000" true "AutoRelease" false))
      (Transformation|AttachComponentToComponent _returnvalue_2 _returnvalue "Muzzle" "KeepRelative" "KeepRelative" "KeepRelative" true)
      (Variables|Default|SetLeftMuzzle(Empty) _returnvalue_2)
      (|CallParentFunction Persistent%20Fire OffHand)
      (else
        (bind _returnvalue_3 (Effects|SpawnEmitteratLocation "/Game/Gadgets/Assets/Turret_Gun/Particles/P_TurretGun_MFlash.P_TurretGun_MFlash" _returnvalue_location _returnvalue_rotation "1.000000,1.000000,1.000000" true "AutoRelease" false))
        (Transformation|AttachComponentToComponent _returnvalue_3 _weaponmesh "Muzzle" "KeepRelative" "KeepRelative" "KeepRelative" true)
        (Variables|Effects|SetMuzzle(Empty) _returnvalue_3)))))

(event EventActivateShellsFX
  (params (Bool :type Boolean) (OffHand :type Boolean))
  (bind (_returnvalue_location _returnvalue_rotation) (Math|Transform|BreakTransform (Transformation|GetSocketTransform (CallFunction|GetLeftHandWeaponMesh) "shells" "RTS_World")))
  (bind (_returnvalue_location_2 _returnvalue_rotation_2) (Math|Transform|BreakTransform (Transformation|GetSocketTransform (Variables|Visual|GetWeaponMesh) "shells" "RTS_World")))
  ;[1 1 1] Muzzle FX
  (if (Variables|Default|GetSkipCustomMuzzleFXSpawnScript)
    (|CallParentFunction Bool OffHand)
    (elif OffHand
      (bind _returnvalue (Effects|SpawnEmitteratLocation "/Game/Weapons/FORT_Generic_Effects/ShellEject/P_ShellsFire_Master_01.P_ShellsFire_Master_01" _returnvalue_location _returnvalue_rotation "1.000000,1.000000,1.000000" true "AutoRelease" false))
      (Variables|Default|SetLeftShells(Empty) _returnvalue)
      (|CallParentFunction Bool OffHand)
      (else
        (bind _returnvalue_2 (Effects|SpawnEmitteratLocation "/Game/Weapons/FORT_Generic_Effects/ShellEject/P_ShellsFire_Master_01.P_ShellsFire_Master_01" _returnvalue_location_2 _returnvalue_rotation_2 "1.000000,1.000000,1.000000" true "AutoRelease" false))
        (Variables|Effects%28Shells%29|SetShells%28empty%29 _returnvalue_2)))))

(fn ConstructionScript
  (|CallParentFunction))
