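; Floor trap that places a drivable mounted turret, spawned by the server and destroying the trap when it is destroyed.
; On attach it hides the floor piece, disables its collision and corrects its own height.
; Placement clears roofs and stairs in the cell, launches overlapping players and destroys chests.
; End play removes the turret and either restores or destroys the floor; damage from its own turret is zeroed.
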(event Custom|HideAttachedFloor
  (params (BuildingAttachedTo :type Building%20Static%20Mesh%20Object%20Reference))
  (bind _attachedfloor (Variables|MountedTurret|GetAttachedFloor))
  ;[1 1 1] Hide Building Mesh and make invulnerable
  (Utilities|IsValid BuildingAttachedTo
    (:Is%20Valid
      (Rendering|SetActorHiddenInGame BuildingAttachedTo true)
      (Variables|/Script/FortniteGame.BuildingSMActor|SetbHiddenDueToTrapPlacement true BuildingAttachedTo)
      (Collision|SetActorEnableCollision BuildingAttachedTo false)
      (Variables|MountedTurret|SetAttachedFloor BuildingAttachedTo)
      ;[1 1 1] HACK: Timer to ensure building piece stays hidden - fixes a bug where the piece becomes visible again for some reason. Only happens when piece is being constructed
      (Utilities|FlowControl|Sequence
        (:then_0
          (Health|AssignOnConstructionComplete _attachedfloor (delegate Custom|Turn%20off%20Tick))
          (Utilities|IsValid _attachedfloor
            (:Is%20Valid
              (Utilities|FlowControl|Delay 0.03)
              (if (Variables|/Script/Engine.Actor|GetbHidden _attachedfloor)
                (if (Variables|MountedTurret|GetShouldTick)
                  _)))))
        (:then_1
          _))
      (Utilities|FlowControl|Delay 25.0)
      (Variables|MountedTurret|SetShouldTick false))
    (:Is%20Not%20Valid
      (Development|PrintString "Warning: No valid attached floor for this!"))))

(event Custom|Turn%20off%20Tick
  (params (BuildingSMActor :type Building%20Static%20Mesh%20Object%20Reference))
  ;[1 1 1] HACK: Timer to ensure building piece stays hidden - fixes a bug where the piece becomes visible again for some reason. Only happens when piece is being constructed
  (Utilities|FlowControl|Delay 2.0)
  (Variables|MountedTurret|SetShouldTick false))

(event EventOnPlaced
  ;[1 1 1] EVENT -> On Placed
  (Utilities|FlowControl|DoOnce false
    (:Completed
      (Audio|PlaySoundatLocation (Variables|Audio|GetPlace_Trap_Sound) (Transformation|GetActorLocation))
      (Utilities|TrapPlacementDustFX :Attach%20to%20Component (Variables|Debug|GetStaticMeshComponent) :Particle%20System%20%28Override%29 "/Game/Gadgets/Assets/Turret_Gun/Particles/P_TurretGun_BuildPoint.P_TurretGun_BuildPoint" :Socket "PlacedFX" :Rotation "0, 0, 0" :Burst%20Rate 15 :Size%20%28Volume%29 "200.000000,200.000000,15.000000" :Velocity%20Scale "1,1,1" :Size%20Scale%20%28Particles%29 1
        (:Out
          (CallFunction|BumpPlayerandDestroyBuildings))))))

(event EventOnDeathServer
  (params (Damage :type Float%20%28single-precision%29) (DamageTags :type Gameplay%20Tag%20Container%20Structure%20%28by%20ref%29) (Momentum :type Vector :default "0, 0, 0") (HitInfo :type Hit%20Result%20Structure%20%28by%20ref%29) (InstigatedBy :type Controller%20Object%20Reference) (DamageCauser :type Actor%20Object%20Reference) (EffectContext :type Gameplay%20Effect%20Context%20Handle%20Structure))
  (bind _mountedturret (Variables|MountedTurret|GetMountedTurret))
  ;[1 1 1] Bind to Destruction of Turret and Base
  (Variables|Default|SetSilentDeath (GameplayTags|HasAnyTags DamageTags (Variables|Default|GetSilentDeathTag)))
  (Utilities|IsValid _mountedturret
    (:Is%20Valid
      (if (Vehicle|HasDriver _mountedturret)
        (GameplayEffects|RemoveActiveEffectswithTags (Ability|GetAbilitySystemComponent (Vehicle|GetDriver _mountedturret)) "(GameplayTags=((TagName=\"Gameplay.Action.Player.AthenaInVehicle\")))")))))

(event EventOnFinishedBuilding
  ;[1 1 1] EVENT -> Finished building
  (Networking|CanExecuteCosmeticEvents
    (:True
      (Audio|PlaySoundatLocation (Variables|Audio|GetTrap_Active_Sound) (Transformation|GetActorLocation))
      (Fortnite|ShowSoundIndicatorLocalforTeam :Location (Transformation|GetActorLocation) :MaxAudibleDistance (Variables|/Script/Engine.SoundBase|GetMaxDistance (Variables|Audio|GetPlace_Trap_Sound)) :IndicatorType "Gunshot" :Team (Team|GetActorTeam self) :Affiliations (Utilities|Array|MakeArray "Hostile" "Neutral") :Tint "(R=0.000000,G=0.000000,B=0.000000,A=0.000000)" :OverrideIcon "/Game/UI/Foundation/Textures/Icons/Navigation/t_icon_nav_mobile_audioviz_building_32.t_icon_nav_mobile_audioviz_building_32" :AttenuationMaxRangeMultiplier 1.0 :AttenuationIntensityMultiplier 1.0))))

(event EventOnWorldReady
  ;[1 1 1] EVENT -> Finished building
  (Networking|CanExecuteCosmeticEvents
    (:True
      (Audio|PlaySoundatLocation (Variables|Audio|GetTrap_Active_Sound) (Transformation|GetActorLocation))
      (Fortnite|ShowSoundIndicatorLocalforTeam :Location (Transformation|GetActorLocation) :MaxAudibleDistance (Variables|/Script/Engine.SoundBase|GetMaxDistance (Variables|Audio|GetPlace_Trap_Sound)) :IndicatorType "Gunshot" :Team (Team|GetActorTeam self) :Affiliations (Utilities|Array|MakeArray "Hostile" "Neutral") :Tint "(R=0.000000,G=0.000000,B=0.000000,A=0.000000)" :OverrideIcon "/Game/UI/Foundation/Textures/Icons/Navigation/t_icon_nav_mobile_audioviz_building_32.t_icon_nav_mobile_audioviz_building_32" :AttenuationMaxRangeMultiplier 1.0 :AttenuationIntensityMultiplier 1.0))))

(event Attachment|EventOnAttachToBuilding
  (params (AttachedToBuilding :type Building%20Actor%20Object%20Reference))
  ;[1 1 1] Snap location to the tile we are attached to.
  (Networking|SwitchHasAuthority
    (:Authority
      (Utilities|FlowControl|DoOnce false
        (:Completed
          (Utilities|FlowControl|Sequence
            (:then_0
              (bind _asbuilding_static_mesh (Utilities|Casting|CastToBuildingSMActor AttachedToBuilding))
              (CallFunction|HideAttachedFloor _asbuilding_static_mesh))
            (:then_1
              ;;[1 0.4136 0] Overtime the turret through multiple saves could start to shift. We need to correct all past saves from that error.
              (Transformation|SetMobility (Variables|Debug|GetStaticMeshComponent) "Movable")
              (Transformation|SetActorLocation (Math|Vector|MakeVector (.x (Transformation|GetActorLocation)) (.y (Transformation|GetActorLocation)) (.z (Transformation|GetActorLocation AttachedToBuilding))) false true)
              (Transformation|SetMobility (Variables|Debug|GetStaticMeshComponent) "Static")
              (CallFunction|SpawnTurret))))))))

(event EventEndPlay
  (params (EndPlayReason :type EEndPlayReason%20Enum :default "Destroyed"))
  (bind _mountedturret (Variables|MountedTurret|GetMountedTurret))
  ;[1 1 1] Special case with creative tool deletion
  (Networking|SwitchHasAuthority
    (:Authority
      (Utilities|IsValid _mountedturret
        (:Is%20Valid
          (if (Vehicle|HasDriver _mountedturret)
            (GameplayEffects|RemoveActiveEffectswithTags (Ability|GetAbilitySystemComponent (Vehicle|GetDriver _mountedturret)) "(GameplayTags=((TagName=\"Gameplay.Action.Player.AthenaInVehicle\")))")
            (else
              _))
          (if (Variables|Default|GetSilentDeath)
            (CallFunction|UnHideFloorandKillTurret)
            (else
              (CallFunction|DestroyFloorandTurret))))
        (:Is%20Not%20Valid
          _)))
    (:Remote
      (CallFunction|UnHideFloorandKillTurret))))

(event Custom|SpawnTurret
  (bind _self self)
  ;[1 1 1] Spawn Turret
  (if (not (Editor|InEditor))
    (Networking|SwitchHasAuthority
      (:Authority
        (Utilities|IsValid (Variables|MountedTurret|GetMountedTurret)
          (:Is%20Not%20Valid
            (bind _spawnedactor (Fortnite|Utilities|SpawnTransientActor "/Game/Athena/DrivableVehicles/MountedTurret.MountedTurret_C" (Transformation|GetWorldTransform (Variables|Default|GetRoot)) "AlwaysSpawn" _self (Game|GetInstigator)))
            (Class|FortVehicleInterface|SetActorBase _spawnedactor _self 0)
            (bind _output_get (Variables|MountedTurret|SetMountedTurret _spawnedactor))
            (Variables|/Script/FortniteGame.FortMountedTurret|SetTeamIndex (Team|GetActorTeam (Building|GetOwningPlayerState)) _output_get)
            (Game|BindEventtoOnDestroyed (Variables|MountedTurret|GetMountedTurret) (delegate Custom|DestroyedTurret))))))))

(event Custom|DestroyedTurret
  (params (DestroyedActor :type Actor%20Object%20Reference))
  ;[1 1 1] Spawn Turret
  (Actor|DestroyActor))

(event EditorPreview|EventSetupForEditorPreview
  (if (Editor|InEditor)
    (Utilities|IsValid (Variables|/Script/Engine.SkinnedMeshComponent|GetSkeletalMesh (Variables|Default|GetSkeletalMesh))
      (:Is%20Not%20Valid
        (bind _loaded_object (Creative|Loading|AsyncLoadEditorSafe (Variables|Default|GetPreviewSkeletalMesh))
          (:Loaded
            (bind _asskeletal_mesh (Utilities|Casting|CastToSkeletalMesh _loaded_object))
            (Components|SkinnedMesh|SetSkinnedAssetandUpdate (Variables|Default|GetSkeletalMesh) _asskeletal_mesh false)))))))

(event EventBeginPlay
  (bind _skeletalmesh (Variables|Default|GetSkeletalMesh))
  (CallFunction|SpawnTurret)
  (Utilities|IsValid (Variables|/Script/Engine.SkinnedMeshComponent|GetSkeletalMesh _skeletalmesh)
    (:Is%20Valid
      (Components|SkinnedMesh|SetSkinnedAssetandUpdate _skeletalmesh 0 false))))

(fn BumpPlayerAndDestroyBuildings
  (Utilities|FlowControl|Sequence
    (:then_0
      ;[1 1 1] Destroy Buildings In Grid Cell
      (bind _allowed (CallFunction|CanDestroySurroundingBuildings))
      (if _allowed
        (bind _outactorsingridcell (Building|K2GetBuildingActorsinGridCell (Building|GetBuildingStructuralSupportSystem) (Transformation|GetWorldLocation (Variables|Default|GetRoot)) (Utilities|Struct|MakeBuildingGridActorFilter false true "None" false)))
        (for _array_element (.NeighboringCenterCellInfos (Utilities|Struct|BreakBuildingNeighboringActorInfo _outactorsingridcell))
          (if (Variables|/Script/Engine.Actor|GetbCanBeDamaged (Utilities|Struct|BreakNeighboringCenterCellInfo _array_element))
            (bind _asbuilding_roof (Utilities|Casting|CastToBuildingRoof (Utilities|Struct|BreakNeighboringCenterCellInfo _array_element))
              (:then
                (Actor|DestroyActor _asbuilding_roof))
              (:CastFailed
                (bind _asbuilding_stairs (Utilities|Casting|CastToBuildingStairs (Utilities|Struct|BreakNeighboringCenterCellInfo _array_element)))
                (Actor|DestroyActor _asbuilding_stairs)))))))
    (:then_1
      ;[1 1 1] Displace Overlapping Players so Turret can spawn
      (if (Variables|MountedTurret|GetShouldBumpPlayer?)
        (if (> (ScalableFloat|GetValueatLevel (Variables|MountedTurret|GetShouldBump?)) 0.0)
          (Utilities|FlowControl|Sequence
            (:then_0
              (bind _outactors (Collision|BoxOverlapActors (Transformation|GetWorldLocation (Variables|Default|GetCenter)) "75.000000,75.000000,200.000000" (Utilities|Array|MakeArray "ObjectTypeQuery3") "/Script/FortniteGame.FortPawn" :!class /Script/Engine.KismetSystemLibrary))
              (for _array_element_2 _outactors
                (bind _asfort_player_pawn (Utilities|Casting|CastToFortPlayerPawn _array_element_2))
                (Character|LaunchCharacter _asfort_player_pawn (CallFunction|CalculateLaunch _asfort_player_pawn) true true)))
            (:then_1
              ;[1 1 1] Destroy overlapped chest
              (bind _outactors_2 (Collision|BoxOverlapActors (Transformation|GetWorldLocation (Variables|Default|GetCenter)) "75.000000,75.000000,200.000000" (Utilities|Array|MakeArray "ObjectTypeQuery1") "/Script/FortniteGame.BuildingContainer" :!class /Script/Engine.KismetSystemLibrary))
              (for _array_element_3 _outactors_2
                (bind _astiered_chest_athena (Utilities|Casting|CastToTiered_Chest_Athena _array_element_3))
                (Actor|DestroyActor _astiered_chest_athena)))))))))

(fn CalculateLaunch
  (params (Player :type Actor%20Object%20Reference))
  (return (Math|Vector|MakeVector (.x (* (Math|Vector|Normalize (- (Transformation|GetActorLocation Player) (Transformation|GetWorldLocation (Variables|Default|GetCenter)))) 385.0)) (.y (* (Math|Vector|Normalize (- (Transformation|GetActorLocation Player) (Transformation|GetWorldLocation (Variables|Default|GetCenter)))) 385.0)) 5e+02)))

(fn CanDestroySurroundingBuildings
  (bind _self self)
  (bind _returnvalue (Minigame|GetMinigame _self :!class /Script/FortniteGame.MinigameBlueprintLibrary))
  (bind _returnvalue_2 (Transformation|GetActorLocation))
  (bind _outvolumemanager (Volume|GetVolumeManager)
    (:Valid
      (Utilities|IsValid (VolumeManager|GetVolumeforLocation _outvolumemanager _returnvalue_2)
        (:Is%20Valid
          (Utilities|IsValid _returnvalue
            (:Is%20Valid
              (if (Minigame|IsinProgress _returnvalue)
                (return true)
                (else
                  _)))
            (:Is%20Not%20Valid
              _))
          (return (Class|FortVolume|IsReady (VolumeManager|GetVolumeforLocation _outvolumemanager _returnvalue_2))))
        (:Is%20Not%20Valid
          _)))
    (:Invalid
      _))
  (return true))

(fn DestroyFloorAndTurret
  (bind _mountedturret (Variables|MountedTurret|GetMountedTurret))
  (bind _attachedfloor (Variables|MountedTurret|GetAttachedFloor))
  (Utilities|IsValid _mountedturret
    (:Is%20Valid
      (Actor|DestroyActor _mountedturret))
    (:Is%20Not%20Valid
      _))
  (Utilities|IsValid _attachedfloor
    (:Is%20Valid
      (if (not (GameplayTags|MatchesAnyTags (Variables|Default|GetCreativePersistOnDetachmentGameplayTag) (GameplayTags|GetOwnedGameplayTags (Variables|MountedTurret|GetAttachedFloor) :!class /Script/GameplayTags.GameplayTagAssetInterface) true))
        (Actor|DestroyActor _attachedfloor)
        (else
          _))
      (Variables|MountedTurret|SetAttachedFloor 0)
      (Variables|MountedTurret|SetMountedTurret 0))
    (:Is%20Not%20Valid
      _)))

(fn GetNumberOfRemainingDelayedAttachingAttempts
  (return 20))

(fn OnModifyTrapIncomingDamage
  (params (Damage :type Float%20%28single-precision%29) (EventInstigator :type Controller%20Object%20Reference) (DamageCauser :type Actor%20Object%20Reference))
  (bind _asb_mounted_turret_athena (Utilities|Casting|CastToB_MountedTurretAthena DamageCauser))
  (if (Utilities|Equal(Object) (Utilities|ToObject(Interface) (Vehicle|GetHostVehicle _asb_mounted_turret_athena)) (Variables|MountedTurret|GetMountedTurret))
    (return 0.0 true)
    (else
      (return Damage false))))

(fn UnHideFloorAndKillTurret
  (bind _attachedfloor (Variables|MountedTurret|GetAttachedFloor))
  (bind _mountedturret (Variables|MountedTurret|GetMountedTurret))
  (Utilities|IsValid _mountedturret
    (:Is%20Valid
      (Actor|DestroyActor _mountedturret))
    (:Is%20Not%20Valid
      _))
  (Variables|MountedTurret|SetMountedTurret 0)
  (Utilities|IsValid _attachedfloor
    (:Is%20Valid
      (Rendering|SetActorHiddenInGame _attachedfloor false)
      (Variables|/Script/FortniteGame.BuildingSMActor|SetbHiddenDueToTrapPlacement false _attachedfloor)
      (Collision|SetActorEnableCollision _attachedfloor (not (Variables|/Script/FortniteGame.BuildingActor|GetbDestroyed _attachedfloor)))
      (Variables|MountedTurret|SetAttachedFloor 0))))

(fn ConstructionScript
  (|CallParentFunction))
